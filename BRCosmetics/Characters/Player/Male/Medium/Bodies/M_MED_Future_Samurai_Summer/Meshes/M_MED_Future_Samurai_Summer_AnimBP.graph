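# Animation Blueprint for the Future Samurai Summer male medium body cosmetic, adding secondary motion and corrective shaping to its ropes, belts and cloth.
# It copies the attached parent mesh's pose, then runs a long component-space chain of bone constraints, additive bone offsets and spring-based Anim Dynamics.
# Variable alphas drive thigh and calf correctives, gate spine rope dynamics by front-end state, and switch left belt swing limits for crouching.
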AnimBlueprint /BRCosmetics/Characters/Player/Male/Medium/Bodies/M_MED_Future_Samurai_Summer/Meshes/M_MED_Future_Samurai_Summer_AnimBP
target_skeleton=&/M_MED_Future_Samurai_Summer_Skeleton
graphs [
 AnimGraph {nodes [
   n1 = "Constraint - dyn_hand_r_rope (RR)"!(
    ComponentPose=Local To Component(LocalPose=Copy Pose From Mesh() {bUseAttachedParent=true})
    Alpha=1
    ConstraintWeights_0=0.5
    ConstraintWeights_1:float=0.5
   ) -> (Pose) {
    alphaBoolBlend {blendOption=Linear}
    boneToModify {boneName=dyn_hand_r_rope}
    constraintSetup [{transformType=Rotation, targetBone {boneName=lowerarm_r}}, {transformType=Rotation, targetBone {boneName=root}}]
    constraintWeights [1,1]
   }
   n2 = "Constraint - dyn_hand_r_rope (PP)"!(ComponentPose=@n1, Alpha=1, ConstraintWeights_0=0.7, ConstraintWeights_1:float=0.3) -> (Pose) {
    alphaBoolBlend {blendOption=Linear}
    boneToModify {boneName=dyn_hand_r_rope}
    constraintSetup [{targetBone {boneName=hand_r}}, {targetBone {boneName=dyn_hand_r_rope}}]
    constraintWeights [1,1]
   }
   n3 = Anim Dynamics - Bone: dyn_hand_r_rope(ComponentPose=@n2, Alpha=0.85, ExternalForce="0,0,0") {
    angularSpringConstant=800
    bAngularSpring=true
    bGravityOverrideInSimSpace=true
    bOverrideAngularDamping=true
    bUseGravityOverride=true
    gravityScale=0.1
    simulationSpace=RootRelative
    alphaBoolBlend {blendOption=Linear}
    boundBone {boneName=dyn_hand_r_rope}
    componentAppliedLinearAccClamp=<1200 1200 1200>
    componentLinearAccScale=<0.2 0.2 0.25>
    componentLinearVelScale=<0.25 0.25 0.25>
    gravityOverride {z=-980}
    physicsBodyDefinitions [{
      sphereCollisionRadius=0
      boundBone {boneName=dyn_hand_r_rope}
      boxExtents=<25 25 25>
      constraintSetup {twistAxis=AxisY, angularLimitsMax=<45 45 60>, angularLimitsMin=<-45 -45 -60>, angularTarget=<0 0 0>}
      localJointOffset=<0.1 -5 0.1>
     }]
    relativeSpaceBone {boneName=root}
   }
   n4 = Anim Dynamics - Bone: dyn_hand_r_rope_01(ComponentPose=@n3, Alpha=0.85, ExternalForce="0,0,0") {
    angularSpringConstant=800
    bAngularSpring=true
    bGravityOverrideInSimSpace=true
    bOverrideAngularDamping=true
    bUseGravityOverride=true
    gravityScale=0.1
    simulationSpace=RootRelative
    alphaBoolBlend {blendOption=Linear}
    boundBone {boneName=dyn_hand_r_rope_01}
    componentAppliedLinearAccClamp=<1200 1200 1200>
    componentLinearAccScale=<0.2 0.2 0.25>
    componentLinearVelScale=<0.25 0.25 0.25>
    gravityOverride {z=-980}
    physicsBodyDefinitions [{
      sphereCollisionRadius=0
      boundBone {boneName=dyn_hand_r_rope_01}
      boxExtents=<25 25 25>
      constraintSetup {twistAxis=AxisY, angularLimitsMax=<12 12 15>, angularLimitsMin=<-12 -12 -15>, angularTarget=<0 0 0>}
      localJointOffset=<0.1 -5 0.1>
     }]
    relativeSpaceBone {boneName=root}
   }
   n5 = Anim Dynamics - Bone: dyn_hand_r_rope_02(ComponentPose=@n4, Alpha=0.85, ExternalForce="0,0,0") {
    angularSpringConstant=800
    bAngularSpring=true
    bGravityOverrideInSimSpace=true
    bOverrideAngularDamping=true
    bUseGravityOverride=true
    gravityScale=0.1
    simulationSpace=RootRelative
    alphaBoolBlend {blendOption=Linear}
    boundBone {boneName=dyn_hand_r_rope_02}
    componentAppliedLinearAccClamp=<1200 1200 1200>
    componentLinearAccScale=<0.2 0.2 0.25>
    componentLinearVelScale=<0.25 0.25 0.25>
    gravityOverride {z=-980}
    physicsBodyDefinitions [{
      sphereCollisionRadius=0
      boundBone {boneName=dyn_hand_r_rope_02}
      boxExtents=<25 25 25>
      constraintSetup {twistAxis=AxisY, angularLimitsMax=<12 12 15>, angularLimitsMin=<-12 -12 -15>, angularTarget=<0 0 0>}
      localJointOffset=<0.1 -5 0.1>
     }]
    relativeSpaceBone {boneName=root}
   }
   n6 = "Constraint - dyn_spine_03_rope (PP)"!(ComponentPose=@n5, Alpha=1, ConstraintWeights_0=0.3, ConstraintWeights_1:float=0.7) -> (Pose) {
    alphaBoolBlend {blendOption=Linear}
    boneToModify {boneName=dyn_spine_03_rope}
    constraintSetup [{targetBone {boneName=spine_02}}, {targetBone {boneName=spine_03}}]
    constraintWeights [1,1]
   }
   n7 = "Transform (Modify) Bone - Bone: dyn_spine_03_rope"(ComponentPose=@n6, Alpha=Get IsInFrontEnd_0() -> (IsInFrontEnd_0:double), Translation="0,0,0", Rotation="0,-20,0", Scale="1,1,1") {
    rotationMode=BMM_Additive
    rotationSpace=BCS_BoneSpace
    scaleMode=BMM_Additive
    scaleSpace=BCS_BoneSpace
    translationMode=BMM_Additive
    translationSpace=BCS_BoneSpace
    alphaBoolBlend {blendOption=Linear}
    boneToModify {boneName=dyn_spine_03_rope}
    rotation {yaw=-20}
   }
   n8 = "Transform (Modify) Bone - Bone: dyn_spine_03_rope_01"(ComponentPose=@n7, Alpha=1, Translation="0,0,0", Rotation="0,0,0", Scale="1,1,1") {
    rotationMode=BMM_Additive
    rotationSpace=BCS_BoneSpace
    scaleMode=BMM_Additive
    scaleSpace=BCS_BoneSpace
    translationMode=BMM_Additive
    translationSpace=BCS_BoneSpace
    alphaBoolBlend {blendOption=Linear}
    boneToModify {boneName=dyn_spine_03_rope_01}
   }
   n9 = Get IsInFrontEnd() -> (IsInFrontEnd:double)
   n10 = Anim Dynamics - Bone: dyn_spine_03_rope(ComponentPose=@n8, Alpha=@n9, ExternalForce="0,0,0") {
    angularSpringConstant=800
    bAngularSpring=true
    bGravityOverrideInSimSpace=true
    bOverrideAngularDamping=true
    bUseGravityOverride=true
    gravityScale=0.3
    simulationSpace=BoneRelative
    alphaBoolBlend {blendOption=Linear}
    boundBone {boneName=dyn_spine_03_rope}
    componentAppliedLinearAccClamp=<1200 1200 1200>
    componentLinearAccScale=<0.2 0.2 0.25>
    componentLinearVelScale=<0.25 0.25 0.25>
    gravityOverride {z=-500}
    physicsBodyDefinitions [{
      sphereCollisionRadius=0
      boundBone {boneName=dyn_spine_03_rope}
      boxExtents=<5 10 10>
      constraintSetup {angularLimitsMax=<2 7 7>, angularLimitsMin=<-2 -7 -7>, angularTarget=<0 0 0>}
      localJointOffset=<-5 1 0.1>
     }]
    relativeSpaceBone {boneName=root}
   }
   n11 = Anim Dynamics - Bone: dyn_spine_03_rope_01(ComponentPose=@n10, Alpha=@n9, ExternalForce="0,0,0") {
    angularSpringConstant=800
    bAngularSpring=true
    bGravityOverrideInSimSpace=true
    bOverrideAngularDamping=true
    bUseGravityOverride=true
    gravityScale=0.3
    simulationSpace=BoneRelative
    alphaBoolBlend {blendOption=Linear}
    boundBone {boneName=dyn_spine_03_rope_01}
    componentAppliedLinearAccClamp=<1200 1200 1200>
    componentLinearAccScale=<0.2 0.2 0.25>
    componentLinearVelScale=<0.25 0.25 0.25>
    gravityOverride {z=-500}
    physicsBodyDefinitions [{
      sphereCollisionRadius=0
      boundBone {boneName=dyn_spine_03_rope_01}
      boxExtents=<5 10 10>
      constraintSetup {angularLimitsMax=<2 7 7>, angularLimitsMin=<-2 -7 -7>, angularTarget=<0 0 0>}
      localJointOffset=<-5 1 0.1>
     }]
    relativeSpaceBone {boneName=root}
   }
   n12 = Anim Dynamics - Bone: dyn_spine_03_rope_02(ComponentPose=@n11, Alpha=@n9, ExternalForce="0,0,0") {
    angularSpringConstant=800
    bAngularSpring=true
    bGravityOverrideInSimSpace=true
    bOverrideAngularDamping=true
    bUseGravityOverride=true
    gravityScale=0.3
    simulationSpace=BoneRelative
    alphaBoolBlend {blendOption=Linear}
    boundBone {boneName=dyn_spine_03_rope_02}
    componentAppliedLinearAccClamp=<1200 1200 1200>
    componentLinearAccScale=<0.2 0.2 0.25>
    componentLinearVelScale=<0.25 0.25 0.25>
    gravityOverride {z=-500}
    physicsBodyDefinitions [{
      sphereCollisionRadius=0
      boundBone {boneName=dyn_spine_03_rope_02}
      boxExtents=<5 10 10>
      constraintSetup {angularLimitsMax=<2 7 7>, angularLimitsMin=<-2 -7 -7>, angularTarget=<0 0 0>}
      localJointOffset=<-5 1 0.1>
     }]
    relativeSpaceBone {boneName=root}
   }
   n13 = "Constraint - thigh_volume_bk_l (PP)"!(ComponentPose=@n12, Alpha=1, ConstraintWeights_0=0, ConstraintWeights_1:float=1) -> (Pose) {
    alphaBoolBlend {blendOption=Linear}
    boneToModify {boneName=thigh_volume_bk_l}
    constraintSetup [{targetBone {boneName=thigh_twist_01_l}}, {targetBone {boneName=pelvis}}]
    constraintWeights [1,1]
   }
   n14 = "Constraint - thigh_volume_fn_l (PP)"!(ComponentPose=@n13, Alpha=Get thigh_shape_fn_l() -> (thigh_shape_fn_l:double), ConstraintWeights_0=0.5, ConstraintWeights_1:float=0.5) -> (Pose) {
    alphaBoolBlend {blendOption=Linear}
    boneToModify {boneName=thigh_volume_fn_l}
    constraintSetup [{targetBone {boneName=thigh_twist_01_l}}, {targetBone {boneName=pelvis}}]
    constraintWeights [1,1]
   }
   n15 = "Constraint - thigh_volume_lt_l (PP)"!(ComponentPose=@n14, Alpha=1, ConstraintWeights_0=0.7, ConstraintWeights_1:float=0.3) -> (Pose) {
    alphaBoolBlend {blendOption=Linear}
    boneToModify {boneName=thigh_volume_lt_l}
    constraintSetup [{targetBone {boneName=thigh_twist_01_l}}, {targetBone {boneName=pelvis}}]
    constraintWeights [1,1]
   }
   n16 = "Constraint - thigh_volume_bk_r (PP)"!(ComponentPose=@n15, Alpha=1, ConstraintWeights_0=0, ConstraintWeights_1:float=1) -> (Pose) {
    alphaBoolBlend {blendOption=Linear}
    boneToModify {boneName=thigh_volume_bk_r}
    constraintSetup [{targetBone {boneName=thigh_twist_01_r}}, {targetBone {boneName=pelvis}}]
    constraintWeights [1,1]
   }
   n17 = "Constraint - thigh_Tshirt_bk_l (PP)"!(ComponentPose=@n16, Alpha=1, ConstraintWeights_0=0, ConstraintWeights_1:float=1) -> (Pose) {
    alphaBoolBlend {blendOption=Linear}
    boneToModify {boneName=thigh_Tshirt_bk_l}
    constraintSetup [{targetBone {boneName=thigh_twist_01_l}}, {targetBone {boneName=pelvis}}]
    constraintWeights [1,1]
   }
   n18 = "Constraint - thigh_volume_fn_r (PP)"!(ComponentPose=@n17, Alpha=Get thigh_shape_fn_r() -> (thigh_shape_fn_r:double), ConstraintWeights_0=0.5, ConstraintWeights_1:float=0.5) -> (Pose) {
    alphaBoolBlend {blendOption=Linear}
    boneToModify {boneName=thigh_volume_fn_r}
    constraintSetup [{targetBone {boneName=thigh_twist_01_r}}, {targetBone {boneName=pelvis}}]
    constraintWeights [1,1]
   }
   n19 = "Constraint - thigh_Tshirt_bk_r (PP)"!(ComponentPose=@n18, Alpha=1, ConstraintWeights_0=0, ConstraintWeights_1:float=1) -> (Pose) {
    alphaBoolBlend {blendOption=Linear}
    boneToModify {boneName=thigh_Tshirt_bk_r}
    constraintSetup [{targetBone {boneName=thigh_twist_01_r}}, {targetBone {boneName=pelvis}}]
    constraintWeights [1,1]
   }
   n20 = "Constraint - thigh_volume_rt_r (PP)"!(ComponentPose=@n19, Alpha=1, ConstraintWeights_0=0.7, ConstraintWeights_1:float=0.3) -> (Pose) {
    alphaBoolBlend {blendOption=Linear}
    boneToModify {boneName=thigh_volume_rt_r}
    constraintSetup [{targetBone {boneName=thigh_twist_01_r}}, {targetBone {boneName=pelvis}}]
    constraintWeights [1,1]
   }
   n21 = "Transform (Modify) Bone - Bone: thigh_volume_fn_l"(ComponentPose=@n20, Alpha=Get thigh_shape_fn_l(), Translation="-10,-30,5", Rotation="0,10,0", Scale="1,1,1") {
    rotationMode=BMM_Additive
    rotationSpace=BCS_BoneSpace
    scaleMode=BMM_Additive
    scaleSpace=BCS_BoneSpace
    translationMode=BMM_Additive
    translationSpace=BCS_BoneSpace
    alphaBoolBlend {blendOption=Linear}
    boneToModify {boneName=thigh_volume_fn_l}
    rotation {yaw=10}
    translation=<-10 -30 5>
   }
   n22 = "Transform (Modify) Bone - Bone: thigh_volume_fn_r"(ComponentPose=@n21, Alpha=Get thigh_shape_fn_r(), Translation="10,-30,5", Rotation="0,10,0", Scale="1,1,1") {
    rotationMode=BMM_Additive
    rotationSpace=BCS_BoneSpace
    scaleMode=BMM_Additive
    scaleSpace=BCS_BoneSpace
    translationMode=BMM_Additive
    translationSpace=BCS_BoneSpace
    alphaBoolBlend {blendOption=Linear}
    boneToModify {boneName=thigh_volume_fn_r}
    rotation {yaw=10}
    translation=<10 -30 5>
   }
   n23 = "Transform (Modify) Bone - Bone: thigh_cloth_fn_l_belt"(ComponentPose=@n22, Alpha=Get thigh_shape_fn_l(), Translation="-5,12,-0.2", Rotation="-50,30,30", Scale="1,1,1") {
    rotationMode=BMM_Additive
    rotationSpace=BCS_BoneSpace
    scaleMode=BMM_Additive
    scaleSpace=BCS_BoneSpace
    translationMode=BMM_Additive
    translationSpace=BCS_BoneSpace
    alphaBoolBlend {blendOption=Linear}
    boneToModify {boneName=thigh_cloth_fn_l_belt}
    rotation=<rot -50 30 30>
    translation=<-5 12 -0.2>
   }
   n24 = "Transform (Modify) Bone - Bone: thigh_cloth_fn_r_belt"(ComponentPose=@n23, Alpha=Get thigh_shape_fn_r(), Translation="5,8,2", Rotation="50,-40,-30", Scale="1,1,1") {
    rotationMode=BMM_Additive
    rotationSpace=BCS_BoneSpace
    scaleMode=BMM_Additive
    scaleSpace=BCS_BoneSpace
    translationMode=BMM_Additive
    translationSpace=BCS_BoneSpace
    alphaBoolBlend {blendOption=Linear}
    boneToModify {boneName=thigh_cloth_fn_r_belt}
    rotation=<rot 50 -40 -30>
    translation=<5 8 2>
   }
   n25 = "Transform (Modify) Bone - Bone: thigh_volume_bk_l"(ComponentPose=@n24, Alpha=Get thigh_shape_bk_l() -> (thigh_shape_bk_l:double), Translation="-5,-3,0", Rotation="0,0,0", Scale="1,1,1") {
    rotationMode=BMM_Additive
    rotationSpace=BCS_BoneSpace
    scaleMode=BMM_Additive
    scaleSpace=BCS_BoneSpace
    translationMode=BMM_Additive
    translationSpace=BCS_BoneSpace
    alphaBoolBlend {blendOption=Linear}
    boneToModify {boneName=thigh_volume_bk_l}
    translation=<-5 -3>
   }
   n26 = "Transform (Modify) Bone - Bone: thigh_volume_bk_r"(ComponentPose=@n25, Alpha=Get thigh_shape_bk_r() -> (thigh_shape_bk_r:double), Translation="5,-3,0", Rotation="0,0,0", Scale="1,1,1") {
    rotationMode=BMM_Additive
    rotationSpace=BCS_BoneSpace
    scaleMode=BMM_Additive
    scaleSpace=BCS_BoneSpace
    translationMode=BMM_Additive
    translationSpace=BCS_BoneSpace
    alphaBoolBlend {blendOption=Linear}
    boneToModify {boneName=thigh_volume_bk_r}
    translation=<5 -3>
   }
   n27 = "Transform (Modify) Bone - Bone: thigh_volume_bk_l"(ComponentPose=@n26, Alpha=Get thigh_shape_bk_l_0() -> (thigh_shape_bk_l_0:double), Translation="0,0,0", Rotation="0,60,0", Scale="1,1,1") {
    rotationMode=BMM_Additive
    rotationSpace=BCS_BoneSpace
    scaleMode=BMM_Additive
    scaleSpace=BCS_BoneSpace
    translationMode=BMM_Additive
    translationSpace=BCS_BoneSpace
    alphaBoolBlend {blendOption=Linear}
    boneToModify {boneName=thigh_volume_bk_l}
    rotation {yaw=60}
   }
   n28 = "Transform (Modify) Bone - Bone: thigh_volume_bk_r"(ComponentPose=@n27, Alpha=Get thigh_shape_bk_r_0() -> (thigh_shape_bk_r_0:double), Translation="0,0,0", Rotation="0,-60,0", Scale="1,1,1") {
    rotationMode=BMM_Additive
    rotationSpace=BCS_BoneSpace
    scaleMode=BMM_Additive
    scaleSpace=BCS_BoneSpace
    translationMode=BMM_Additive
    translationSpace=BCS_BoneSpace
    alphaBoolBlend {blendOption=Linear}
    boneToModify {boneName=thigh_volume_bk_r}
    rotation {yaw=-60}
   }
   n29 = "Transform (Modify) Bone - Bone: thigh_Tshirt_bk_l"(ComponentPose=@n28, Alpha=Get thigh_shape_bk_l(), Translation="0,-4,0", Rotation="0,0,0", Scale="1,1,1") {
    rotationMode=BMM_Additive
    rotationSpace=BCS_BoneSpace
    scaleMode=BMM_Additive
    scaleSpace=BCS_BoneSpace
    translationMode=BMM_Additive
    translationSpace=BCS_BoneSpace
    alphaBoolBlend {blendOption=Linear}
    boneToModify {boneName=thigh_Tshirt_bk_l}
    translation {y=-4}
   }
   n30 = "Transform (Modify) Bone - Bone: thigh_Tshirt_bk_r"(ComponentPose=@n29, Alpha=Get thigh_shape_bk_r(), Translation="-1,-4,0", Rotation="0,0,0", Scale="1,1,1") {
    rotationMode=BMM_Additive
    rotationSpace=BCS_BoneSpace
    scaleMode=BMM_Additive
    scaleSpace=BCS_BoneSpace
    translationMode=BMM_Additive
    translationSpace=BCS_BoneSpace
    alphaBoolBlend {blendOption=Linear}
    boneToModify {boneName=thigh_Tshirt_bk_r}
    translation=<-1 -4>
   }
   n31 = "Transform (Modify) Bone - Bone: thigh_volume_bk_l_belt"(ComponentPose=@n30, Alpha=Get thigh_shape_bk_l(), Translation="4,-1,0", Rotation="0,0,0", Scale="1,1,1") {
    rotationMode=BMM_Additive
    rotationSpace=BCS_BoneSpace
    scaleMode=BMM_Additive
    scaleSpace=BCS_BoneSpace
    translationMode=BMM_Additive
    translationSpace=BCS_BoneSpace
    alphaBoolBlend {blendOption=Linear}
    boneToModify {boneName=thigh_volume_bk_l_belt}
    translation=<4 -1>
   }
   n32 = "Transform (Modify) Bone - Bone: thigh_volume_bk_r_belt"(ComponentPose=@n31, Alpha=Get thigh_shape_bk_r(), Translation="4,-1,0", Rotation="0,0,0", Scale="1,1,1") {
    rotationMode=BMM_Additive
    rotationSpace=BCS_BoneSpace
    scaleMode=BMM_Additive
    scaleSpace=BCS_BoneSpace
    translationMode=BMM_Additive
    translationSpace=BCS_BoneSpace
    alphaBoolBlend {blendOption=Linear}
    boneToModify {boneName=thigh_volume_bk_r_belt}
    translation=<4 -1>
   }
   n33 = "Transform (Modify) Bone - Bone: thigh_volume_lt_l"(ComponentPose=@n32, Alpha=Get thigh_shape_lt_l() -> (thigh_shape_lt_l:double), Translation="0,-1,0", Rotation="0,0,0", Scale="1,1,1") {
    rotationMode=BMM_Additive
    rotationSpace=BCS_BoneSpace
    scaleMode=BMM_Additive
    scaleSpace=BCS_BoneSpace
    translationMode=BMM_Additive
    translationSpace=BCS_BoneSpace
    alphaBoolBlend {blendOption=Linear}
    boneToModify {boneName=thigh_volume_lt_l}
    translation {y=-1}
   }
   n34 = "Transform (Modify) Bone - Bone: thigh_volume_rt_r"(ComponentPose=@n33, Alpha=Get thigh_shape_rt_r() -> (thigh_shape_rt_r:double), Translation="0,-1,0", Rotation="0,0,0", Scale="1,1,1") {
    rotationMode=BMM_Additive
    rotationSpace=BCS_BoneSpace
    scaleMode=BMM_Additive
    scaleSpace=BCS_BoneSpace
    translationMode=BMM_Additive
    translationSpace=BCS_BoneSpace
    alphaBoolBlend {blendOption=Linear}
    boneToModify {boneName=thigh_volume_rt_r}
    translation {y=-1}
   }
   n35 = "Transform (Modify) Bone - Bone: dyn_thigh_lt_l_belt_01"(ComponentPose=@n34, Alpha=Get thigh_shape_lt_l(), Translation="2,0,0", Rotation="-15,0,20", Scale="1,1,1") {
    rotationMode=BMM_Additive
    rotationSpace=BCS_BoneSpace
    scaleMode=BMM_Additive
    scaleSpace=BCS_BoneSpace
    translationMode=BMM_Additive
    translationSpace=BCS_BoneSpace
    alphaBoolBlend {blendOption=Linear}
    boneToModify {boneName=dyn_thigh_lt_l_belt_01}
    rotation {pitch=-15, roll=20}
    translation {x=2}
   }
   n36 = "Transform (Modify) Bone - Bone: dyn_thigh_rt_r_belt_01"(ComponentPose=@n35, Alpha=Get thigh_shape_rt_r(), Translation="2,0,0", Rotation="-15,0,-20", Scale="1,1,1") {
    rotationMode=BMM_Additive
    rotationSpace=BCS_BoneSpace
    scaleMode=BMM_Additive
    scaleSpace=BCS_BoneSpace
    translationMode=BMM_Additive
    translationSpace=BCS_BoneSpace
    alphaBoolBlend {blendOption=Linear}
    boneToModify {boneName=dyn_thigh_rt_r_belt_01}
    rotation {pitch=-15, roll=-20}
    translation {x=2}
   }
   n37 = "Transform (Modify) Bone - Bone: dyn_thigh_lt_l_belt_01"(ComponentPose=@n36, Alpha=1, Translation="0,0,0", Rotation="0,20,0", Scale="1,1,1") {
    rotationMode=BMM_Additive
    rotationSpace=BCS_BoneSpace
    scaleMode=BMM_Additive
    scaleSpace=BCS_BoneSpace
    translationMode=BMM_Additive
    translationSpace=BCS_BoneSpace
    alphaBoolBlend {blendOption=Linear}
    boneToModify {boneName=dyn_thigh_lt_l_belt_01}
    rotation {yaw=20}
   }
   n38 = "Transform (Modify) Bone - Bone: dyn_thigh_rt_r_belt_01"(ComponentPose=@n37, Alpha=1, Translation="0,0,0", Rotation="0,20,0", Scale="1,1,1") {
    rotationMode=BMM_Additive
    rotationSpace=BCS_BoneSpace
    scaleMode=BMM_Additive
    scaleSpace=BCS_BoneSpace
    translationMode=BMM_Additive
    translationSpace=BCS_BoneSpace
    alphaBoolBlend {blendOption=Linear}
    boneToModify {boneName=dyn_thigh_rt_r_belt_01}
    rotation {yaw=20}
   }
   n39 = Anim Dynamics - Bone: dyn_thigh_lt_l_belt_01(ComponentPose=@n38, Alpha=Get Not_crouch() -> (Not_crouch:double), ExternalForce="0,0,0") {
    angularSpringConstant=800
    bAngularSpring=true
    bGravityOverrideInSimSpace=true
    bOverrideAngularDamping=true
    bUseGravityOverride=true
    gravityScale=0.3
    simulationSpace=BoneRelative
    alphaBoolBlend {blendOption=Linear}
    boundBone {boneName=dyn_thigh_lt_l_belt_01}
    componentAppliedLinearAccClamp=<1200 1200 1200>
    componentLinearAccScale=<0.2 0.2 0.25>
    componentLinearVelScale=<0.25 0.25 0.25>
    gravityOverride {z=-500}
    physicsBodyDefinitions [{
      sphereCollisionRadius=0
      boundBone {boneName=dyn_thigh_lt_l_belt_01}
      boxExtents=<5 10 10>
      constraintSetup {angularLimitsMax=<0.1 10 60>, angularLimitsMin=<-0.1 -10 -20>, angularTarget=<0 0 0>}
      localJointOffset=<-5 1 0.1>
     }]
    relativeSpaceBone {boneName=root}
   }
   n40 = Anim Dynamics - Bone: dyn_thigh_lt_l_belt_01(ComponentPose=@n39, Alpha=Get crouch() -> (crouch:double), ExternalForce="0,0,0") {
    angularSpringConstant=800
    bAngularSpring=true
    bGravityOverrideInSimSpace=true
    bOverrideAngularDamping=true
    bUseGravityOverride=true
    gravityScale=0.3
    simulationSpace=BoneRelative
    alphaBoolBlend {blendOption=Linear}
    boundBone {boneName=dyn_thigh_lt_l_belt_01}
    componentAppliedLinearAccClamp=<1200 1200 1200>
    componentLinearAccScale=<0.2 0.2 0.25>
    componentLinearVelScale=<0.25 0.25 0.25>
    gravityOverride {z=-500}
    physicsBodyDefinitions [{
      sphereCollisionRadius=0
      boundBone {boneName=dyn_thigh_lt_l_belt_01}
      boxExtents=<5 10 10>
      constraintSetup {angularLimitsMax=<0.1 10 5>, angularLimitsMin=<-0.1 -10 -5>, angularTarget=<0 0 0>}
      localJointOffset=<-5 1 0.1>
     }]
    relativeSpaceBone {boneName=root}
   }
   n41 = Anim Dynamics - Bone: dyn_thigh_lt_l_belt_02(ComponentPose=@n40, Alpha=0.7, ExternalForce="0,0,0") {
    angularSpringConstant=800
    bAngularSpring=true
    bGravityOverrideInSimSpace=true
    bOverrideAngularDamping=true
    bUseGravityOverride=true
    gravityScale=0.3
    simulationSpace=BoneRelative
    alphaBoolBlend {blendOption=Linear}
    boundBone {boneName=dyn_thigh_lt_l_belt_02}
    componentAppliedLinearAccClamp=<1200 1200 1200>
    componentLinearAccScale=<0.2 0.2 0.25>
    componentLinearVelScale=<0.25 0.25 0.25>
    gravityOverride {z=-500}
    physicsBodyDefinitions [{
      sphereCollisionRadius=0
      boundBone {boneName=dyn_thigh_lt_l_belt_02}
      boxExtents=<15 15 15>
      constraintSetup {angularLimitsMax=<0.1 5 30>, angularLimitsMin=<-0.1 -5 -20>, angularTarget=<0 0 0>}
      localJointOffset=<-3 0.1 0.1>
     }]
    relativeSpaceBone {boneName=root}
   }
   n42 = Anim Dynamics - Bone: dyn_thigh_rt_r_belt_01(ComponentPose=@n41, Alpha=0.7, ExternalForce="0,0,0") {
    angularSpringConstant=800
    bAngularSpring=true
    bGravityOverrideInSimSpace=true
    bOverrideAngularDamping=true
    bUseGravityOverride=true
    gravityScale=0.3
    simulationSpace=BoneRelative
    alphaBoolBlend {blendOption=Linear}
    boundBone {boneName=dyn_thigh_rt_r_belt_01}
    componentAppliedLinearAccClamp=<1200 1200 1200>
    componentLinearAccScale=<0.2 0.2 0.25>
    componentLinearVelScale=<0.25 0.25 0.25>
    gravityOverride {z=-500}
    physicsBodyDefinitions [{
      sphereCollisionRadius=0
      boundBone {boneName=dyn_thigh_rt_r_belt_01}
      boxExtents=<5 10 10>
      constraintSetup {angularLimitsMax=<0.1 5 45>, angularLimitsMin=<-0.1 -5 -10>, angularTarget=<0 0 0>}
      localJointOffset=<-5 1 0.1>
     }]
    relativeSpaceBone {boneName=root}
   }
   n43 = Anim Dynamics - Bone: dyn_thigh_rt_r_belt_02(ComponentPose=@n42, Alpha=0.7, ExternalForce="0,0,0") {
    angularSpringConstant=800
    bAngularSpring=true
    bGravityOverrideInSimSpace=true
    bOverrideAngularDamping=true
    bUseGravityOverride=true
    gravityScale=0.3
    simulationSpace=BoneRelative
    alphaBoolBlend {blendOption=Linear}
    boundBone {boneName=dyn_thigh_rt_r_belt_02}
    componentAppliedLinearAccClamp=<1200 1200 1200>
    componentLinearAccScale=<0.2 0.2 0.25>
    componentLinearVelScale=<0.25 0.25 0.25>
    gravityOverride {z=-500}
    physicsBodyDefinitions [{
      sphereCollisionRadius=0
      boundBone {boneName=dyn_thigh_rt_r_belt_02}
      boxExtents=<15 15 15>
      constraintSetup {angularLimitsMax=<0.1 5 30>, angularLimitsMin=<-0.1 -5 -20>, angularTarget=<0 0 0>}
      localJointOffset=<-3 0.1 0.1>
     }]
    relativeSpaceBone {boneName=root}
   }
   n44 = Anim Dynamics - Bone: dyn_thigh_l_cloth(ComponentPose=@n43, Alpha=0.7, ExternalForce="0,0,0") {
    angularSpringConstant=800
    bAngularSpring=true
    bGravityOverrideInSimSpace=true
    bOverrideAngularDamping=true
    bUseGravityOverride=true
    gravityScale=0.3
    simulationSpace=BoneRelative
    alphaBoolBlend {blendOption=Linear}
    boundBone {boneName=dyn_thigh_l_cloth}
    componentAppliedLinearAccClamp=<1200 1200 1200>
    componentLinearAccScale=<0.2 0.2 0.25>
    componentLinearVelScale=<0.25 0.25 0.25>
    gravityOverride {z=-500}
    physicsBodyDefinitions [{
      sphereCollisionRadius=0
      boundBone {boneName=dyn_thigh_l_cloth}
      boxExtents=<12 15 15>
      constraintSetup {angularLimitsMax=<0.1 15 13>, angularLimitsMin=<-0.1 -50 -13>, angularTarget=<0 0 0>}
      localJointOffset=<-5 1 -1>
     }]
    relativeSpaceBone {boneName=root}
   }
   n45 = Anim Dynamics - Bone: dyn_thigh_r_cloth(ComponentPose=@n44, Alpha=0.7, ExternalForce="0,0,0") {
    angularSpringConstant=800
    bAngularSpring=true
    bGravityOverrideInSimSpace=true
    bOverrideAngularDamping=true
    bUseGravityOverride=true
    gravityScale=0.3
    simulationSpace=BoneRelative
    alphaBoolBlend {blendOption=Linear}
    boundBone {boneName=dyn_thigh_r_cloth}
    componentAppliedLinearAccClamp=<1200 1200 1200>
    componentLinearAccScale=<0.2 0.2 0.25>
    componentLinearVelScale=<0.25 0.25 0.25>
    gravityOverride {z=-500}
    physicsBodyDefinitions [{
      sphereCollisionRadius=0
      boundBone {boneName=dyn_thigh_r_cloth}
      boxExtents=<12 15 15>
      constraintSetup {angularLimitsMax=<0.1 15 10>, angularLimitsMin=<-0.1 -15 -10>, angularTarget=<0 0 0>}
      localJointOffset=<-5 1 1.2>
     }]
    relativeSpaceBone {boneName=root}
   }
   n46 = "Transform (Modify) Bone - Bone: dyn_thigh_cloth_fn_l_rope"(ComponentPose=@n45, Alpha=1, Translation="0,0,0", Rotation="0,10,0", Scale="1,1,1") {
    rotationMode=BMM_Additive
    rotationSpace=BCS_BoneSpace
    scaleMode=BMM_Additive
    scaleSpace=BCS_BoneSpace
    translationMode=BMM_Additive
    translationSpace=BCS_BoneSpace
    alphaBoolBlend {blendOption=Linear}
    boneToModify {boneName=dyn_thigh_cloth_fn_l_rope}
    rotation {yaw=10}
   }
   n47 = "Transform (Modify) Bone - Bone: dyn_thigh_volume_bk_l_rope"(ComponentPose=@n46, Alpha=1, Translation="0,0,0", Rotation="0,10,0", Scale="1,1,1") {
    rotationMode=BMM_Additive
    rotationSpace=BCS_BoneSpace
    scaleMode=BMM_Additive
    scaleSpace=BCS_BoneSpace
    translationMode=BMM_Additive
    translationSpace=BCS_BoneSpace
    alphaBoolBlend {blendOption=Linear}
    boneToModify {boneName=dyn_thigh_volume_bk_l_rope}
    rotation {yaw=10}
   }
   n48 = "Transform (Modify) Bone - Bone: dyn_thigh_cloth_fn_r_rope"(ComponentPose=@n47, Alpha=1, Translation="0,0,0", Rotation="0,10,0", Scale="1,1,1") {
    rotationMode=BMM_Additive
    rotationSpace=BCS_BoneSpace
    scaleMode=BMM_Additive
    scaleSpace=BCS_BoneSpace
    translationMode=BMM_Additive
    translationSpace=BCS_BoneSpace
    alphaBoolBlend {blendOption=Linear}
    boneToModify {boneName=dyn_thigh_cloth_fn_r_rope}
    rotation {yaw=10}
   }
   n49 = "Transform (Modify) Bone - Bone: dyn_thigh_volume_bk_r_rope"(ComponentPose=@n48, Alpha=1, Translation="0,0,0", Rotation="0,10,0", Scale="1,1,1") {
    rotationMode=BMM_Additive
    rotationSpace=BCS_BoneSpace
    scaleMode=BMM_Additive
    scaleSpace=BCS_BoneSpace
    translationMode=BMM_Additive
    translationSpace=BCS_BoneSpace
    alphaBoolBlend {blendOption=Linear}
    boneToModify {boneName=dyn_thigh_volume_bk_r_rope}
    rotation {yaw=10}
   }
   n50 = Anim Dynamics - Bone: dyn_thigh_cloth_fn_l_rope(ComponentPose=@n49, Alpha=1, ExternalForce="0,0,0") {
    angularSpringConstant=800
    bAngularSpring=true
    bGravityOverrideInSimSpace=true
    bOverrideAngularDamping=true
    bUseGravityOverride=true
    gravityScale=0.7
    simulationSpace=RootRelative
    alphaBoolBlend {blendOption=Linear}
    boundBone {boneName=dyn_thigh_cloth_fn_l_rope}
    componentAppliedLinearAccClamp=<300 300 300>
    componentLinearAccScale=<0.1 0.1 0.15>
    componentLinearVelScale=<0.1 0.1 0.15>
    gravityOverride {z=-500}
    physicsBodyDefinitions [{
      sphereCollisionRadius=0
      boundBone {boneName=dyn_thigh_cloth_fn_l_rope}
      boxExtents=<5 7 7>
      constraintSetup {angularLimitsMax=<1 1 60>, angularLimitsMin=<-1 -1 -20>, angularTarget=<0 0 30>}
      localJointOffset=<-5 -1 0.1>
     }]
    relativeSpaceBone {boneName=pelvis}
   }
   n51 = Anim Dynamics - Bone: dyn_thigh_volume_bk_l_rope(ComponentPose=@n50, Alpha=1, ExternalForce="0,0,0") {
    angularSpringConstant=800
    bAngularSpring=true
    bGravityOverrideInSimSpace=true
    bOverrideAngularDamping=true
    bUseGravityOverride=true
    gravityScale=0.7
    simulationSpace=RootRelative
    alphaBoolBlend {blendOption=Linear}
    boundBone {boneName=dyn_thigh_volume_bk_l_rope}
    componentAppliedLinearAccClamp=<300 300 300>
    componentLinearAccScale=<0.1 0.1 0.15>
    componentLinearVelScale=<0.1 0.1 0.15>
    gravityOverride {z=-500}
    physicsBodyDefinitions [{
      sphereCollisionRadius=0
      boundBone {boneName=dyn_thigh_volume_bk_l_rope}
      boxExtents=<5 7 7>
      constraintSetup {angularLimitsMax=<1 1 60>, angularLimitsMin=<-1 -1 -15>, angularTarget=<0 0 30>}
      localJointOffset=<-5 -1.5 0.1>
     }]
    relativeSpaceBone {boneName=pelvis}
   }
   n52 = Anim Dynamics - Bone: dyn_thigh_cloth_fn_r_rope(ComponentPose=@n51, Alpha=1, ExternalForce="0,0,0") {
    angularSpringConstant=800
    bAngularSpring=true
    bGravityOverrideInSimSpace=true
    bOverrideAngularDamping=true
    bUseGravityOverride=true
    gravityScale=0.7
    simulationSpace=RootRelative
    alphaBoolBlend {blendOption=Linear}
    boundBone {boneName=dyn_thigh_cloth_fn_r_rope}
    componentAppliedLinearAccClamp=<300 300 300>
    componentLinearAccScale=<0.1 0.1 0.15>
    componentLinearVelScale=<0.1 0.1 0.15>
    gravityOverride {z=-500}
    physicsBodyDefinitions [{
      sphereCollisionRadius=0
      boundBone {boneName=dyn_thigh_cloth_fn_r_rope}
      boxExtents=<5 7 7>
      constraintSetup {angularLimitsMax=<1 1 60>, angularLimitsMin=<-1 -1 -20>, angularTarget=<0 0 30>}
      localJointOffset=<-5 -1 0.1>
     }]
    relativeSpaceBone {boneName=pelvis}
   }
   n53 = Anim Dynamics - Bone: dyn_thigh_volume_bk_r_rope(ComponentPose=@n52, Alpha=1, ExternalForce="0,0,0") {
    angularSpringConstant=800
    bAngularSpring=true
    bGravityOverrideInSimSpace=true
    bOverrideAngularDamping=true
    bUseGravityOverride=true
    gravityScale=0.7
    simulationSpace=RootRelative
    alphaBoolBlend {blendOption=Linear}
    boundBone {boneName=dyn_thigh_volume_bk_r_rope}
    componentAppliedLinearAccClamp=<300 300 300>
    componentLinearAccScale=<0.1 0.1 0.15>
    componentLinearVelScale=<0.1 0.1 0.15>
    gravityOverride {z=-500}
    physicsBodyDefinitions [{
      sphereCollisionRadius=0
      boundBone {boneName=dyn_thigh_volume_bk_r_rope}
      boxExtents=<5 7 7>
      constraintSetup {angularLimitsMax=<1 1 60>, angularLimitsMin=<-1 -1 -15>, angularTarget=<0 0 30>}
      localJointOffset=<-5 -1.5 0.1>
     }]
    relativeSpaceBone {boneName=pelvis}
   }
   n54 = "Constraint - calf_r_front (RR)"!(ComponentPose=@n53, Alpha=1, ConstraintWeights_0=0.5, ConstraintWeights_1:float=0.5) -> (Pose) {
    alphaBoolBlend {blendOption=Linear}
    boneToModify {boneName=calf_r_front}
    constraintSetup [{transformType=Rotation, targetBone {boneName=thigh_l}}, {transformType=Rotation, targetBone {boneName=calf_l}}]
    constraintWeights [1,1]
   }
   n55 = "Constraint - calf_r_front (RR)"(ComponentPose=@n54, Alpha=1, ConstraintWeights_0=0.5, ConstraintWeights_1=0.5) {
    alphaBoolBlend {blendOption=Linear}
    boneToModify {boneName=calf_r_front}
    constraintSetup [{transformType=Rotation, targetBone {boneName=thigh_r}}, {transformType=Rotation, targetBone {boneName=calf_r}}]
    constraintWeights [1,1]
   }
   n56 = "Constraint - calf_l_back (RR)"!(ComponentPose=@n55, Alpha=1, ConstraintWeights_0=0.5, ConstraintWeights_1:float=0.5) -> (Pose) {
    alphaBoolBlend {blendOption=Linear}
    boneToModify {boneName=calf_l_back}
    constraintSetup [{transformType=Rotation, targetBone {boneName=thigh_l}}, {transformType=Rotation, targetBone {boneName=calf_l}}]
    constraintWeights [1,1]
   }
   n57 = "Constraint - calf_r_back (RR)"!(ComponentPose=@n56, Alpha=1, ConstraintWeights_0=0.5, ConstraintWeights_1:float=0.5) -> (Pose) {
    alphaBoolBlend {blendOption=Linear}
    boneToModify {boneName=calf_r_back}
    constraintSetup [{transformType=Rotation, targetBone {boneName=thigh_r}}, {transformType=Rotation, targetBone {boneName=calf_r}}]
    constraintWeights [1,1]
   }
   n58 = Get calf_l_front() -> (calf_l_front:double)
   n59 = "Transform (Modify) Bone - Bone: calf_l_front"(ComponentPose=@n57, Alpha=@n58, Translation="0,1.5,0", Rotation="0,0,0", Scale="1,1,1") {
    rotationMode=BMM_Additive
    rotationSpace=BCS_BoneSpace
    scaleMode=BMM_Additive
    scaleSpace=BCS_BoneSpace
    translationMode=BMM_Additive
    translationSpace=BCS_BoneSpace
    alphaBoolBlend {blendOption=Linear}
    boneToModify {boneName=calf_l_front}
    translation {y=1.5}
   }
   n60 = Get calf_r_front() -> (calf_r_front:double)
   n61 = "Transform (Modify) Bone - Bone: calf_r_front"(ComponentPose=@n59, Alpha=@n60, Translation="0,-1.5,0", Rotation="0,0,0", Scale="1,1,1") {
    rotationMode=BMM_Additive
    rotationSpace=BCS_BoneSpace
    scaleMode=BMM_Additive
    scaleSpace=BCS_BoneSpace
    translationMode=BMM_Additive
    translationSpace=BCS_BoneSpace
    alphaBoolBlend {blendOption=Linear}
    boneToModify {boneName=calf_r_front}
    translation {y=-1.5}
   }
   n62 = "Transform (Modify) Bone - Bone: calf_l_back"(ComponentPose=@n61, Alpha=@n58, Translation="0,-12,0", Rotation="0,0,0", Scale="1,1,1") {
    rotationMode=BMM_Additive
    rotationSpace=BCS_BoneSpace
    scaleMode=BMM_Additive
    scaleSpace=BCS_BoneSpace
    translationMode=BMM_Additive
    translationSpace=BCS_BoneSpace
    alphaBoolBlend {blendOption=Linear}
    boneToModify {boneName=calf_l_back}
    translation {y=-12}
   }
   n63 = "Transform (Modify) Bone - Bone: calf_r_back"(ComponentPose=@n62, Alpha=@n60, Translation="0,12,0", Rotation="0,0,0", Scale="1,1,1") {
    rotationMode=BMM_Additive
    rotationSpace=BCS_BoneSpace
    scaleMode=BMM_Additive
    scaleSpace=BCS_BoneSpace
    translationMode=BMM_Additive
    translationSpace=BCS_BoneSpace
    alphaBoolBlend {blendOption=Linear}
    boneToModify {boneName=calf_r_back}
    translation {y=12}
   }
   n64 = "Transform (Modify) Bone - Bone: dyn_calf_l_rope"(ComponentPose=@n63, Alpha=1, Translation="0,0,0", Rotation="10,-10,0", Scale="1,1,1") {
    rotationMode=BMM_Additive
    rotationSpace=BCS_BoneSpace
    scaleMode=BMM_Additive
    scaleSpace=BCS_BoneSpace
    translationMode=BMM_Additive
    translationSpace=BCS_BoneSpace
    alphaBoolBlend {blendOption=Linear}
    boneToModify {boneName=dyn_calf_l_rope}
    rotation {pitch=10, yaw=-10}
   }
   n65 = "Transform (Modify) Bone - Bone: dyn_calf_r_rope"(ComponentPose=@n64, Alpha=1, Translation="0,0,0", Rotation="-10,-10,0", Scale="1,1,1") {
    rotationMode=BMM_Additive
    rotationSpace=BCS_BoneSpace
    scaleMode=BMM_Additive
    scaleSpace=BCS_BoneSpace
    translationMode=BMM_Additive
    translationSpace=BCS_BoneSpace
    alphaBoolBlend {blendOption=Linear}
    boneToModify {boneName=dyn_calf_r_rope}
    rotation {pitch=-10, yaw=-10}
   }
   n66 = Anim Dynamics - Bone: dyn_calf_l_rope(ComponentPose=@n65, Alpha=0.7, ExternalForce="0,0,0") {
    angularSpringConstant=800
    bAngularSpring=true
    bGravityOverrideInSimSpace=true
    bOverrideAngularDamping=true
    bUseGravityOverride=true
    gravityScale=0.3
    simulationSpace=RootRelative
    alphaBoolBlend {blendOption=Linear}
    boundBone {boneName=dyn_calf_l_rope}
    componentAppliedLinearAccClamp=<1200 1200 1200>
    componentLinearAccScale=<0.2 0.2 0.25>
    componentLinearVelScale=<0.25 0.25 0.25>
    gravityOverride {z=-700}
    physicsBodyDefinitions [{
      sphereCollisionRadius=0
      boundBone {boneName=dyn_calf_l_rope}
      boxExtents=<8 17 17>
      constraintSetup {angularLimitsMax=<0.1 50 20>, angularLimitsMin=<-0.1 -60 -30>, angularTarget=<0 0 0>}
      localJointOffset=<-5 5 -5>
     }]
    relativeSpaceBone {boneName=pelvis}
   }
   n67 = Anim Dynamics - Bone: dyn_calf_r_rope(ComponentPose=@n66, Alpha=0.7, ExternalForce="0,0,0") {
    angularSpringConstant=800
    bAngularSpring=true
    bGravityOverrideInSimSpace=true
    bOverrideAngularDamping=true
    bUseGravityOverride=true
    gravityScale=0.3
    simulationSpace=RootRelative
    alphaBoolBlend {blendOption=Linear}
    boundBone {boneName=dyn_calf_r_rope}
    componentAppliedLinearAccClamp=<1200 1200 1200>
    componentLinearAccScale=<0.2 0.2 0.25>
    componentLinearVelScale=<0.25 0.25 0.25>
    gravityOverride {z=-700}
    physicsBodyDefinitions [{
      sphereCollisionRadius=0
      boundBone {boneName=dyn_calf_r_rope}
      boxExtents=<8 17 17>
      constraintSetup {angularLimitsMax=<0.1 60 20>, angularLimitsMin=<-0.1 -50 -30>, angularTarget=<0 0 0>}
      localJointOffset=<-5 5 5>
     }]
    relativeSpaceBone {boneName=pelvis}
   }
   n68 = "Transform (Modify) Bone - Bone: dyn_thigh_lt_l_belt_01"(ComponentPose=@n67, Alpha=Get calf_l_front_0() -> (calf_l_front_0:double), Translation="0,3,0", Rotation="10,20,25", Scale="1,1,1") {
    rotationMode=BMM_Additive
    rotationSpace=BCS_BoneSpace
    scaleMode=BMM_Additive
    scaleSpace=BCS_BoneSpace
    translationMode=BMM_Additive
    translationSpace=BCS_BoneSpace
    alphaBoolBlend {blendOption=Linear}
    boneToModify {boneName=dyn_thigh_lt_l_belt_01}
    rotation=<rot 10 20 25>
    translation {y=3}
   }
   n69 = "Transform (Modify) Bone - Bone: dyn_thigh_rt_r_belt_01"(ComponentPose=@n68, Alpha=Get calf_r_front_0() -> (calf_r_front_0:double), Translation="0,3,0", Rotation="-10,20,25", Scale="1,1,1") {
    rotationMode=BMM_Additive
    rotationSpace=BCS_BoneSpace
    scaleMode=BMM_Additive
    scaleSpace=BCS_BoneSpace
    translationMode=BMM_Additive
    translationSpace=BCS_BoneSpace
    alphaBoolBlend {blendOption=Linear}
    boneToModify {boneName=dyn_thigh_rt_r_belt_01}
    rotation=<rot -10 20 25>
    translation {y=3}
   }
   n70 = "Constraint - flont_uppelalm_l (PP)"!(ComponentPose=@n69, Alpha=1, ConstraintWeights_0=0.5, ConstraintWeights_1:float=0.5) -> (Pose) {
    alphaBoolBlend {blendOption=Linear}
    boneToModify {boneName=flont_uppelalm_l}
    constraintSetup [{targetBone {boneName=upperarm_twist_01_l}}, {targetBone {boneName=clavicle_l}}]
    constraintWeights [1,1]
   }
   n71 = "Constraint - up_uppelalm_l (PP)"!(ComponentPose=@n70, Alpha=1, ConstraintWeights_0=0.5, ConstraintWeights_1:float=0.5) -> (Pose) {
    alphaBoolBlend {blendOption=Linear}
    boneToModify {boneName=up_uppelalm_l}
    constraintSetup [{targetBone {boneName=upperarm_twist_01_l}}, {targetBone {boneName=clavicle_l}}]
    constraintWeights [1,1]
   }
   n72 = "Constraint - front_upperarm_r (PP)"!(ComponentPose=@n71, Alpha=1, ConstraintWeights_0=0.3, ConstraintWeights_1:float=0.7) -> (Pose) {
    alphaBoolBlend {blendOption=Linear}
    boneToModify {boneName=front_upperarm_r}
    constraintSetup [{targetBone {boneName=upperarm_twist_01_r}}, {targetBone {boneName=clavicle_r}}]
    constraintWeights [1,1]
   }
   n73 = "Constraint - up_upperarm_r (PP)"!(ComponentPose=@n72, Alpha=1, ConstraintWeights_0=0.5, ConstraintWeights_1:float=0.5) -> (Pose) {
    alphaBoolBlend {blendOption=Linear}
    boneToModify {boneName=up_upperarm_r}
    constraintSetup [{targetBone {boneName=upperarm_twist_01_r}}, {targetBone {boneName=clavicle_r}}]
    constraintWeights [1,1]
   }
   n74 = "Transform (Modify) Bone - Bone: flont_uppelalm_l"(ComponentPose=@n73, Alpha=Get Shoulder_rot_f_l() -> (Shoulder_rot_f_l:double), Translation="-5,5,-5", Rotation="0,0,0", Scale="1,1,1") {
    rotationMode=BMM_Additive
    rotationSpace=BCS_BoneSpace
    scaleMode=BMM_Additive
    scaleSpace=BCS_BoneSpace
    translationMode=BMM_Additive
    translationSpace=BCS_BoneSpace
    alphaBoolBlend {blendOption=Linear}
    boneToModify {boneName=flont_uppelalm_l}
    translation=<-5 5 -5>
   }
   n75 = "Transform (Modify) Bone - Bone: front_upperarm_r"(ComponentPose=@n74, Alpha=Get Shoulder_rot_f_r() -> (Shoulder_rot_f_r:double), Translation="10,-12,-5", Rotation="0,0,0", Scale="2,1,1") {
    rotationMode=BMM_Additive
    rotationSpace=BCS_BoneSpace
    scaleMode=BMM_Additive
    scaleSpace=BCS_BoneSpace
    translationMode=BMM_Additive
    translationSpace=BCS_BoneSpace
    alphaBoolBlend {blendOption=Linear}
    boneToModify {boneName=front_upperarm_r}
    scale {x=2}
    translation=<10 -12 -5>
   }
   n76 = "Transform (Modify) Bone - Bone: up_uppelalm_l"(ComponentPose=@n75, Alpha=Get Shoulder_rot_up_l() -> (Shoulder_rot_up_l:double), Translation="-15,0,2", Rotation="0,0,0", Scale="1,1,1") {
    rotationMode=BMM_Additive
    rotationSpace=BCS_BoneSpace
    scaleMode=BMM_Additive
    scaleSpace=BCS_BoneSpace
    translationMode=BMM_Additive
    translationSpace=BCS_BoneSpace
    alphaBoolBlend {blendOption=Linear}
    boneToModify {boneName=up_uppelalm_l}
    translation {x=-15, z=2}
   }
   n77 = "Transform (Modify) Bone - Bone: up_upperarm_r"(ComponentPose=@n76, Alpha=Get Shoulder_rot_up_r_0() -> (Shoulder_rot_up_r_0:double), Translation="5,-9,-15", Rotation="0,0,0", Scale="1,1,1") {
    rotationMode=BMM_Additive
    rotationSpace=BCS_BoneSpace
    scaleMode=BMM_Additive
    scaleSpace=BCS_BoneSpace
    translationMode=BMM_Additive
    translationSpace=BCS_BoneSpace
    alphaBoolBlend {blendOption=Linear}
    boneToModify {boneName=up_upperarm_r}
    translation=<5 -9 -15>
   }
   n78 = "Transform (Modify) Bone - Bone: upperarmDn_r"(ComponentPose=@n77, Alpha=Get Shoulder_rot_dn_r() -> (Shoulder_rot_dn_r:double), Translation="4,0,0", Rotation="0,0,0", Scale="1,1,1") {
    rotationMode=BMM_Additive
    rotationSpace=BCS_BoneSpace
    scaleMode=BMM_Additive
    scaleSpace=BCS_BoneSpace
    translationMode=BMM_Additive
    translationSpace=BCS_BoneSpace
    alphaBoolBlend {blendOption=Linear}
    boneToModify {boneName=upperarmDn_r}
    translation {x=4}
   }
   n79 = "Transform (Modify) Bone - Bone: upperarmDn_l"(ComponentPose=@n78, Alpha=Get Shoulder_rot_dn_l() -> (Shoulder_rot_dn_l:double), Translation="-4,0,0", Rotation="0,0,0", Scale="1,1,1") {
    rotationMode=BMM_Additive
    rotationSpace=BCS_BoneSpace
    scaleMode=BMM_Additive
    scaleSpace=BCS_BoneSpace
    translationMode=BMM_Additive
    translationSpace=BCS_BoneSpace
    alphaBoolBlend {blendOption=Linear}
    boneToModify {boneName=upperarmDn_l}
    translation {x=-4}
   }
   n80 = "Transform (Modify) Bone - Bone: upperarmDn_r"(ComponentPose=@n79, Alpha=Get Shoulder_rot_up_r() -> (Shoulder_rot_up_r:double), Translation="0,-2,-10", Rotation="0,0,0", Scale="1,1,1") {
    rotationMode=BMM_Additive
    rotationSpace=BCS_BoneSpace
    scaleMode=BMM_Additive
    scaleSpace=BCS_BoneSpace
    translationMode=BMM_Additive
    translationSpace=BCS_BoneSpace
    alphaBoolBlend {blendOption=Linear}
    boneToModify {boneName=upperarmDn_r}
    translation {y=-2, z=-10}
   }
   n81 = "Transform (Modify) Bone - Bone: upperarmDn_l"(ComponentPose=@n80, Alpha=Get Shoulder_rot_up_l(), Translation="0,2,10", Rotation="0,0,0", Scale="1,1,1") {
    rotationMode=BMM_Additive
    rotationSpace=BCS_BoneSpace
    scaleMode=BMM_Additive
    scaleSpace=BCS_BoneSpace
    translationMode=BMM_Additive
    translationSpace=BCS_BoneSpace
    alphaBoolBlend {blendOption=Linear}
    boneToModify {boneName=upperarmDn_l}
    translation {y=2, z=10}
   }
   n82 = Anim Dynamics - Bone: dyn_upperarm_l_cloth(ComponentPose=@n81, Alpha=0.7, ExternalForce="0,0,0") {
    angularSpringConstant=800
    bAngularSpring=true
    bGravityOverrideInSimSpace=true
    bOverrideAngularDamping=true
    bUseGravityOverride=true
    gravityScale=0.3
    simulationSpace=BoneRelative
    alphaBoolBlend {blendOption=Linear}
    boundBone {boneName=dyn_upperarm_l_cloth}
    componentAppliedLinearAccClamp=<1200 1200 1200>
    componentLinearAccScale=<0.2 0.2 0.2>
    componentLinearVelScale=<0.2 0.2 0.2>
    gravityOverride {z=-500}
    physicsBodyDefinitions [{
      sphereCollisionRadius=0
      boundBone {boneName=dyn_upperarm_l_cloth}
      boxExtents=<5 10 10>
      constraintSetup {angularLimitsMax=<2 1 5>, angularLimitsMin=<-2 -1 -5>, angularTarget=<0 0 0>}
      localJointOffset=<-5 1 0.1>
     }]
    relativeSpaceBone {boneName=root}
   }
   n83 = Anim Dynamics - Bone: dyn_upperarm_l_cloth_01(ComponentPose=@n82, Alpha=0.7, ExternalForce="0,0,0") {
    angularSpringConstant=800
    bAngularSpring=true
    bGravityOverrideInSimSpace=true
    bOverrideAngularDamping=true
    bUseGravityOverride=true
    gravityScale=0.3
    simulationSpace=BoneRelative
    alphaBoolBlend {blendOption=Linear}
    boundBone {boneName=dyn_upperarm_l_cloth_01}
    componentAppliedLinearAccClamp=<1200 1200 1200>
    componentLinearAccScale=<0.2 0.2 0.2>
    componentLinearVelScale=<0.2 0.2 0.2>
    gravityOverride {z=-500}
    physicsBodyDefinitions [{
      sphereCollisionRadius=0
      boundBone {boneName=dyn_upperarm_l_cloth_01}
      boxExtents=<15 15 15>
      constraintSetup {angularLimitsMax=<2 1 5>, angularLimitsMin=<-2 -1 -5>, angularTarget=<0 0 0>}
      localJointOffset=<-3 0.1 0.1>
     }]
    relativeSpaceBone {boneName=root}
   }
   n84 = Anim Dynamics - Bone: dyn_upperarm_r_cloth(ComponentPose=@n83, Alpha=0.7, ExternalForce="0,0,0") {
    angularSpringConstant=800
    bAngularSpring=true
    bGravityOverrideInSimSpace=true
    bOverrideAngularDamping=true
    bUseGravityOverride=true
    gravityScale=0.3
    simulationSpace=BoneRelative
    alphaBoolBlend {blendOption=Linear}
    boundBone {boneName=dyn_upperarm_r_cloth}
    componentAppliedLinearAccClamp=<1200 1200 1200>
    componentLinearAccScale=<0.2 0.2 0.2>
    componentLinearVelScale=<0.2 0.2 0.2>
    gravityOverride {z=-500}
    physicsBodyDefinitions [{
      sphereCollisionRadius=0
      boundBone {boneName=dyn_upperarm_r_cloth}
      boxExtents=<5 10 10>
      constraintSetup {angularLimitsMax=<2 1 5>, angularLimitsMin=<-2 -1 -5>, angularTarget=<0 0 0>}
      localJointOffset=<-5 1 0.1>
     }]
    relativeSpaceBone {boneName=root}
   }
   n85 = Anim Dynamics - Bone: dyn_upperarm_r_cloth_01(ComponentPose=@n84, Alpha=0.7, ExternalForce="0,0,0") {
    angularSpringConstant=800
    bAngularSpring=true
    bGravityOverrideInSimSpace=true
    bOverrideAngularDamping=true
    bUseGravityOverride=true
    gravityScale=0.3
    simulationSpace=BoneRelative
    alphaBoolBlend {blendOption=Linear}
    boundBone {boneName=dyn_upperarm_r_cloth_01}
    componentAppliedLinearAccClamp=<1200 1200 1200>
    componentLinearAccScale=<0.2 0.2 0.2>
    componentLinearVelScale=<0.2 0.2 0.2>
    gravityOverride {z=-500}
    physicsBodyDefinitions [{
      sphereCollisionRadius=0
      boundBone {boneName=dyn_upperarm_r_cloth_01}
      boxExtents=<15 15 15>
      constraintSetup {angularLimitsMax=<2 1 5>, angularLimitsMin=<-2 -1 -5>, angularTarget=<0 0 0>}
      localJointOffset=<-3 0.1 0.1>
     }]
    relativeSpaceBone {boneName=root}
   }
   n86 = "Constraint - lowerarm_l_front (RR)"!(ComponentPose=@n85, Alpha=1, ConstraintWeights_0=0.5, ConstraintWeights_1:float=0.5) -> (Pose) {
    alphaBoolBlend {blendOption=Linear}
    boneToModify {boneName=lowerarm_l_front}
    constraintSetup [{transformType=Rotation, targetBone {boneName=lowerarm_l}}, {transformType=Rotation, targetBone {boneName=upperarm_l}}]
    constraintWeights [1,1]
   }
   n87 = "Constraint - lowerarm_r_front (RR)"!(ComponentPose=@n86, Alpha=1, ConstraintWeights_0=0.5, ConstraintWeights_1:float=0.5) -> (Pose) {
    alphaBoolBlend {blendOption=Linear}
    boneToModify {boneName=lowerarm_r_front}
    constraintSetup [{transformType=Rotation, targetBone {boneName=lowerarm_r}}, {transformType=Rotation, targetBone {boneName=upperarm_r}}]
    constraintWeights [1,1]
   }
   n88 = "Constraint - lowerarm_l_back (RR)"!(ComponentPose=@n87, Alpha=1, ConstraintWeights_0=0.5, ConstraintWeights_1:float=0.5) -> (Pose) {
    alphaBoolBlend {blendOption=Linear}
    boneToModify {boneName=lowerarm_l_back}
    constraintSetup [{transformType=Rotation, targetBone {boneName=lowerarm_l}}, {transformType=Rotation, targetBone {boneName=upperarm_l}}]
    constraintWeights [1,1]
   }
   n89 = "Constraint - lowerarm_r_back (RR)"!(ComponentPose=@n88, Alpha=1, ConstraintWeights_0=0.5, ConstraintWeights_1:float=0.5) -> (Pose) {
    alphaBoolBlend {blendOption=Linear}
    boneToModify {boneName=lowerarm_r_back}
    constraintSetup [{transformType=Rotation, targetBone {boneName=lowerarm_r}}, {transformType=Rotation, targetBone {boneName=upperarm_r}}]
    constraintWeights [1,1]
   }
   n90 = Get lowerarm_l_front() -> (lowerarm_l_front:double)
   n91 = "Transform (Modify) Bone - Bone: lowerarm_l_front"(ComponentPose=@n89, Alpha=@n90, Translation="5,5,0", Rotation="0,0,0", Scale="1,1,1") {
    rotationMode=BMM_Additive
    rotationSpace=BCS_BoneSpace
    scaleMode=BMM_Additive
    scaleSpace=BCS_BoneSpace
    translationMode=BMM_Additive
    translationSpace=BCS_BoneSpace
    alphaBoolBlend {blendOption=Linear}
    boneToModify {boneName=lowerarm_l_front}
    translation=<5 5>
   }
   n92 = Get lowerarm_r_front() -> (lowerarm_r_front:double)
   n93 = "Transform (Modify) Bone - Bone: lowerarm_r_front"(ComponentPose=@n91, Alpha=@n92, Translation="-5,-5,0", Rotation="0,0,0", Scale="1,1,1") {
    rotationMode=BMM_Additive
    rotationSpace=BCS_BoneSpace
    scaleMode=BMM_Additive
    scaleSpace=BCS_BoneSpace
    translationMode=BMM_Additive
    translationSpace=BCS_BoneSpace
    alphaBoolBlend {blendOption=Linear}
    boneToModify {boneName=lowerarm_r_front}
    translation=<-5 -5>
   }
   n94 = "Transform (Modify) Bone - Bone: lowerarm_l_back"(ComponentPose=@n93, Alpha=@n90, Translation="0,-5,0", Rotation="0,0,0", Scale="1,1,1") {
    rotationMode=BMM_Additive
    rotationSpace=BCS_BoneSpace
    scaleMode=BMM_Additive
    scaleSpace=BCS_BoneSpace
    translationMode=BMM_Additive
    translationSpace=BCS_BoneSpace
    alphaBoolBlend {blendOption=Linear}
    boneToModify {boneName=lowerarm_l_back}
    translation {y=-5}
   }
   n95 = "Transform (Modify) Bone - Bone: lowerarm_r_back"(ComponentPose=@n94, Alpha=@n92, Translation="0,5,0", Rotation="0,0,0", Scale="1,1,1") {
    rotationMode=BMM_Additive
    rotationSpace=BCS_BoneSpace
    scaleMode=BMM_Additive
    scaleSpace=BCS_BoneSpace
    translationMode=BMM_Additive
    translationSpace=BCS_BoneSpace
    alphaBoolBlend {blendOption=Linear}
    boneToModify {boneName=lowerarm_r_back}
    translation {y=5}
   }
   Output Pose(Result=Component To Local(ComponentPose=@n95)) {layerGroup=None}
  ]}
 EventGraph {nodes [
   Event BlueprintInitializeAnimation()
   > Set Skel_Mesh(Skel_Mesh=GetOwnerSkeletalMeshComponent())
   > n1 = Cast To PlayerPawn_Athena(Object=TryGetPawnOwner())
   > n2 = Set Pawn Owner(Pawn Owner:Fort Player Pawn Athena Object Reference=@n1) -> (Output_Get:Fort Player Pawn Athena Object Reference)
   > Set Shoulder_Mesh(Shoulder_Mesh:Skeletal Mesh Component Object Reference=GetSkeletalMeshForPartType(self=@n2)) -> (Output_Get:Skeletal Mesh Component Object Reference)
   Event BlueprintUpdateAnimation()
   > n3 = Sequence!(execute) -> (then_0, then_1, then_2:Exec, then_3:Exec)
   n4 = Get Skel_Mesh()
   > Set thigh_shape_fn_l(thigh_shape_fn_l:double=GetAlphaFromReferenceTransformDelta(SkeletalMeshComponent=@n4, BoneName=thigh_l, BaseName=pelvis, InputRange=(X=0,Y=-140), TwistAxis="0,0,1")) -> (Output_Get:double)
   > Set thigh_shape_fn_r(thigh_shape_fn_r:double=GetAlphaFromReferenceTransformDelta(SkeletalMeshComponent=@n4, BoneName=thigh_r, BaseName=pelvis, InputRange=(X=0,Y=-140), TwistAxis="0,0,1")) -> (Output_Get:double)
   n5 = Get Skel_Mesh()
   > Set thigh_shape_bk_l(thigh_shape_bk_l:double=GetAlphaFromReferenceTransformDelta(SkeletalMeshComponent=@n5, BoneName=thigh_l, BaseName=pelvis, InputRange=(X=0,Y=-110), TwistAxis="0,0,1")) -> (Output_Get:double)
   > Set thigh_shape_bk_r(thigh_shape_bk_r:double=GetAlphaFromReferenceTransformDelta(SkeletalMeshComponent=@n5, BoneName=thigh_r, BaseName=pelvis, InputRange=(X=0,Y=-110), TwistAxis="0,0,1")) -> (Output_Get:double)
   n6 = Get Skel_Mesh()
   > Set thigh_shape_lt_l(thigh_shape_lt_l:double=GetAlphaFromReferenceTransformDelta(SkeletalMeshComponent=@n6, BoneName=thigh_l, BaseName=pelvis, InputRange=(X=-90,Y=0), OutputRange=(X=1,Y=0), TwistAxis="0,0,1")) -> (Output_Get:double)
   > Set thigh_shape_rt_r(thigh_shape_rt_r:double=GetAlphaFromReferenceTransformDelta(SkeletalMeshComponent=@n6, BoneName=thigh_r, BaseName=pelvis, InputRange=(X=-90,Y=0), OutputRange=(X=1,Y=0), TwistAxis="0,0,1")) -> (Output_Get:double)
   n7 = Get Skel_Mesh()
   > Set thigh_shape_bk_l_0(thigh_shape_bk_l_0:double=GetAlphaFromReferenceTransformDelta(SkeletalMeshComponent=@n7, BoneName=thigh_l, BaseName=pelvis, InputRange=(X=0,Y=60), TwistAxis="0,0,1")) -> (Output_Get:double)
   > Set thigh_shape_bk_r_0(thigh_shape_bk_r_0:double=GetAlphaFromReferenceTransformDelta(SkeletalMeshComponent=@n7, BoneName=thigh_r, BaseName=pelvis, InputRange=(X=0,Y=60), TwistAxis="0,0,1")) -> (Output_Get:double)
   n8 = Get Shoulder_Mesh() -> (Shoulder_Mesh:Skeletal Mesh Component Object Reference)
   > Set Shoulder_rot_f_r(Shoulder_rot_f_r:double=GetAlphaFromReferenceTransformDelta(SkeletalMeshComponent=@n8, BoneName=upperarm_r, BaseName=clavicle_r, InputRange=(X=0,Y=180), TwistAxis="0,0,1")) -> (Output_Get:double)
   > Set Shoulder_rot_f_l(Shoulder_rot_f_l:double=GetAlphaFromReferenceTransformDelta(SkeletalMeshComponent=@n8, BoneName=upperarm_l, BaseName=clavicle_l, InputRange=(X=0,Y=90), TwistAxis="0,0,1")) -> (Output_Get:double)
   n9 = Get Shoulder_Mesh()
   > Set Shoulder_rot_up_r(Shoulder_rot_up_r:double=GetAlphaFromReferenceTransformDelta(
     SkeletalMeshComponent=@n9
     BoneName=upperarm_r
     BaseName=spine_05
     DeltaComponentType=EulerY
     InputRange=(X=60,Y=-60)
     OutputRange=(X=1,Y=0)
     TwistAxis="0,0,1"
    )) -> (Output_Get:double)
   > Set Shoulder_rot_up_l(Shoulder_rot_up_l:double=GetAlphaFromReferenceTransformDelta(
     SkeletalMeshComponent=@n9
     BoneName=upperarm_l
     BaseName=spine_05
     DeltaComponentType=EulerY
     InputRange=(X=60,Y=-60)
     OutputRange=(X=1,Y=0)
     TwistAxis="0,0,1"
    )) -> (Output_Get:double)
   n10 = Get Shoulder_Mesh()
   > Set Shoulder_rot_dn_r(Shoulder_rot_dn_r:double=GetAlphaFromReferenceTransformDelta(SkeletalMeshComponent=@n10, BoneName=upperarm_r, BaseName=clavicle_r, DeltaComponentType=EulerX, InputRange=(X=-15,Y=-90), TwistAxis="0,0,1")) -> (Output_Get:double)
   > Set Shoulder_rot_up_r_0(Shoulder_rot_up_r_0:double=GetAlphaFromReferenceTransformDelta(SkeletalMeshComponent=Get Shoulder_Mesh(), BoneName=upperarm_r, BaseName=clavicle_r, InputRange=(X=0,Y=150), TwistAxis="0,0,1")) -> (Output_Get:double)
   > Set Shoulder_rot_dn_l(Shoulder_rot_dn_l:double=GetAlphaFromReferenceTransformDelta(SkeletalMeshComponent=@n10, BoneName=upperarm_l, BaseName=clavicle_l, DeltaComponentType=EulerX, InputRange=(X=-15,Y=-90), TwistAxis="0,0,1")) -> (Output_Get:double)
   > n11 = Sequence()
   > Set lowerarm_l_front(lowerarm_l_front:double=GetAlphaFromReferenceTransformDelta(SkeletalMeshComponent=Get Skel_Mesh(), BoneName=upperarm_l, BaseName=lowerarm_l, InputRange=(X=0,Y=-90), TwistAxis="0,0,1")) -> (Output_Get:double)
   n12 = Get Skel_Mesh()
   > Set calf_l_front(calf_l_front:double=GetAlphaFromReferenceTransformDelta(SkeletalMeshComponent=@n12, BoneName=thigh_l, BaseName=calf_l, InputRange=(X=0,Y=-90), TwistAxis="0,0,1")) -> (Output_Get:double)
   > Set calf_l_front_0(calf_l_front_0:double=GetAlphaFromReferenceTransformDelta(SkeletalMeshComponent=@n12, BoneName=thigh_l, BaseName=calf_l, InputRange=(X=-110,Y=-180), TwistAxis="0,0,1")) -> (Output_Get:double)
   Set lowerarm_r_front(
    execute=@n11.then_1
    lowerarm_r_front:double=GetAlphaFromReferenceTransformDelta(SkeletalMeshComponent=Get Skel_Mesh(), BoneName=upperarm_r, BaseName=lowerarm_r, InputRange=(X=0,Y=-90), TwistAxis="0,0,1")
   ) -> (Output_Get:double)
   n13 = Get Skel_Mesh()
   > Set calf_r_front(calf_r_front:double=GetAlphaFromReferenceTransformDelta(SkeletalMeshComponent=@n13, BoneName=thigh_r, BaseName=calf_r, InputRange=(X=-90,Y=-150), TwistAxis="0,0,1")) -> (Output_Get:double)
   > Set calf_r_front_0(calf_r_front_0:double=GetAlphaFromReferenceTransformDelta(SkeletalMeshComponent=@n13, BoneName=thigh_r, BaseName=calf_r, InputRange=(X=-90,Y=-150), TwistAxis="0,0,1")) -> (Output_Get:double)
   n14 = Branch(execute=@n3.then_1, Condition=Get bIsCrouching())
   > Set crouch(crouch:double=0.7) -> (Output_Get:double)
   > Set Not_crouch(Not_crouch:double=0) -> (Output_Get:double)
   Set crouch(execute=@n14.else, crouch=0)
   > Set Not_crouch(Not_crouch=0.7)
   n15 = Branch(execute=@n3.then_2, Condition=Get bIsInFrontEnd())
   > Set IsInFrontEnd(IsInFrontEnd:double=0) -> (Output_Get:double)
   > Set IsInFrontEnd_0(IsInFrontEnd_0:double=0) -> (Output_Get:double)
   Set IsInFrontEnd(execute=@n15.else, IsInFrontEnd=0.7)
   > Set IsInFrontEnd_0(IsInFrontEnd_0=1)
  ]}
]
variables [
 calf_l_front {type=number}
 calf_l_front_0 {type=number}
 calf_r_front {type=number}
 calf_r_front_0 {type=number}
 crouch {type=number}
 IsInFrontEnd {type=number}
 IsInFrontEnd_0 {type=number}
 IsInSliding {type=number}
 lowerarm_l_back {type=number}
 lowerarm_l_front {type=number}
 lowerarm_r_back {type=number}
 lowerarm_r_front {type=number}
 Not_crouch {type=number}
 Pawn Owner {type=FortPlayerPawnAthena}
 Shoulder_Mesh {type=SkeletalMeshComponent}
 Shoulder_rot_dn_l {type=number}
 Shoulder_rot_dn_r {type=number}
 Shoulder_rot_f_l {type=number}
 Shoulder_rot_f_r {type=number}
 Shoulder_rot_up_l {type=number}
 Shoulder_rot_up_r {type=number}
 Shoulder_rot_up_r_0 {type=number}
 Skel_Mesh {type=SkeletalMeshComponent}
 thigh_shape_bk_l {type=number}
 thigh_shape_bk_l_0 {type=number}
 thigh_shape_bk_r {type=number}
 thigh_shape_bk_r_0 {type=number}
 thigh_shape_fn_l {type=number}
 thigh_shape_fn_r {type=number}
 thigh_shape_lt_l {type=number}
 thigh_shape_rt_r {type=number}
]
